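# Translucent Niagara mesh-particle material for the light cone FX on the RedPepper backpack cosmetic.
# Emissive blends between particle red with added blue and full particle color, then passes through an emissive adjustment function.
# Opacity uses a flippable cone gradient texture with optional sine side falloff, or a sine square gradient, plus optional view-angle fade, all toggled by scalar lerps.
# Particle alpha and dynamic parameters drive blink and depth-fade distance.
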Material /BRCosmetics/Accessories/FORT_Backpacks/Backpack_M_MED_RedPepper/FX/M_Backpack_RedPepper_LightCone
graph {
 properties {bUsedWithNiagaraMeshParticles=true, blendMode=BLEND_Translucent, pixelDepthOffsetMode=PDOM_Legacy, usedShadingModels=Default Lit}
 nodes [
  n1 = ParticleColor()
  n2 = MaterialFunctionCall(Float3=@n1.RGB) {materialFunction=&BreakOutFloat3Components}
  n3 = ScalarParameter() {defaultValue=0.01, parameterName=BlueAddition}
  MaterialFunctionCall(Emissive=LinearInterpolate(
    A=MaterialFunctionCall(R=@n2.R, B=@n3) {materialFunction=&AppendMany}.RGB
    B=ParticleColor().RGB
    Alpha=ScalarParameter() {parameterName=UseParticleColorOnly}
   )) {materialFunction=&MF_FX_EmissiveAdjustments}
  n4 = DynamicParameter() {paramNames [BlinkCone, DepthFadeDistance, Param3, Param4]}
  n5 = TextureCoordinate()
  n6 = ComponentMask(None=@n5) {r=true}
  n7 = ComponentMask(None=@n5) {g=true}
  n8 = LinearInterpolate(A=@n7, B=OneMinus(None=@n7), Alpha=ScalarParameter() {parameterName=FlipGradient_V})
  n9 = Power(
   Base=TextureSample(UVs=AppendVector(A=@n6, B=@n8)) {texture=&/T_FX_RedPepperBackpack_ConeGradient}.RGB
   Exp=ScalarParameter() {defaultValue=1, parameterName=VGradientPower}
  )
  n10 = TextureCoordinate()
  n11 = Sine(None=Multiply(A=ComponentMask(None=@n10) {r=true}) {constB=0.5})
  n12 = ScalarParameter() {defaultValue=1, parameterName=SideGradient_Power}
  n13 = Saturate(None=Multiply(A=Power(Base=@n11, Exp=@n12), B=ScalarParameter() {defaultValue=1, parameterName=SideGradient_Mult}))
  n14 = Sine(None=Multiply(A=TextureCoordinate()) {constB=0.5})
  n15 = Multiply(
   A=Power(Base=ComponentMask(None=@n14) {r=true}, Exp=ScalarParameter() {defaultValue=2, parameterName=HRamp_Power})
   B=ScalarParameter() {defaultValue=1, parameterName=HRamp_Mult}
  )
  n16 = Multiply(
   A=Power(Base=ComponentMask(None=@n14) {g=true}, Exp=ScalarParameter() {defaultValue=2, parameterName=VRamp_Power})
   B=ScalarParameter() {defaultValue=1, parameterName=VRamp_Mult}
  )
  n17 = LinearInterpolate(
   A=LinearInterpolate(A=@n9, B=Multiply(A=@n9, B=@n13), Alpha=ScalarParameter() {parameterName=UseSideGradient})
   B=Multiply(A=Saturate(None=@n15), B=Saturate(None=@n16))
   Alpha=ScalarParameter() {parameterName=UseSquareGradient}
  )
  n18 = ScalarParameter() {defaultValue=1, parameterName=AngleFade}
  n19 = LinearInterpolate(
   A=@n17
   B=Multiply(
    A=@n17
    B=MaterialFunctionCall("FadeSensitivity (optional)"=@n18) {materialFunction=&FadeBasedOnViewAngle}.Result
   )
   Alpha=ScalarParameter() {parameterName=UseAngleFade}
  )
  n20 = Multiply(A=@n4.BlinkCone, B=LinearInterpolate(A=@n19, B=Saturate(None=@n19), Alpha=ScalarParameter() {parameterName=UseSaturate}))
  Multiply(
   A=DepthFade(Opacity=Multiply(A=@n1.A, B=@n20), FadeDistance=@n4.DepthFadeDistance)
   B=ScalarParameter() {defaultValue=1, parameterName=OpacityMult}
  )
 ]
}
