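# Translucent, unlit, two-sided Niagara mesh particle material for the rune effect on the RedPepper backpack cosmetic.
# Distorted primary noise times secondary noise is inverted, shaped by a UV sphere mask, then used as gradient lookup coordinates for HDR color.
# Dynamic parameters drive noise panning, HDR intensity and blink, while a scalar blends to two-color Fresnel emissive.
# Opacity combines Fresnel with panning noise, with a solid-opacity blend.
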Material /BRCosmetics/Accessories/FORT_Backpacks/Backpack_M_MED_RedPepper/FX/M_Backpack_RedPepper_Rune
graph {
 properties {
  bUsedWithNiagaraMeshParticles=true
  blendMode=BLEND_Translucent
  pixelDepthOffsetMode=PDOM_Legacy
  shadingModel=MSM_Unlit
  twoSided=true
  usedShadingModels=Unlit
 }
 nodes [
  Fresnel() {exponent=1}
  n1 = Multiply(A=Fresnel() {exponent=1}) {constB=2}
  n2 = Constant() {r=0.35}
  n3 = TextureCoordinate()
  n4 = VectorParameter() {group=Opacity Noise, parameterName=Opacity Noise UV Tile, sortPriority=2, defaultValue=<rgba 2 2 0 1>}
  n5 = AppendVector(A=@n4.R, B=@n4.R)
  n6 = VectorParameter() {group=Opacity Noise, parameterName=Opacity Noise Pan Speed, defaultValue=<rgba 0 0.1 0 1>}
  n7 = TextureSample(UVs=Panner(Coordinate=Multiply(A=@n3, B=@n5), Time=Time(), Speed=AppendVector(A=@n6.R, B=@n6.R))) {samplerType=SAMPLERTYPE_LinearGrayscale, texture=/BRCosmetics/Weapons/FORT_Melee/Pickaxe_SnowFall_Female/FX/Snowfall_pick_Noise}
  n8 = ScalarParameter() {defaultValue=1, group=Opacity Noise, parameterName=Opacity Noise Power, sortPriority=0}
  n9 = LinearInterpolate(Alpha=Multiply(A=Power(Base=@n7.RGB, Exp=@n8), B=ScalarParameter() {defaultValue=1, group=Opacity Noise, parameterName=Opacity Noise Mult, sortPriority=1})) {constA=0.6}
  LinearInterpolate(A=Multiply(A=Saturate(None=@n1), B=Multiply(A=@n2, B=@n9)), Alpha=ScalarParameter() {parameterName=UseSolidOpacity})
  n10 = VectorParameter() {group=Primary Noise, parameterName=Primary Noise UV Tile, sortPriority=2, defaultValue=<rgba 0.5 0.5 0 1>}
  n11 = Multiply(A=TextureCoordinate(), B=AppendVector(A=@n10.R, B=@n10.R))
  n12 = TextureCoordinate()
  n13 = VectorParameter() {group=Primary Noise Distortion, parameterName=Distortion Noise UV Tile, sortPriority=1, defaultValue=<rgba 1 1 0 1>}
  n14 = AppendVector(A=@n13.R, B=@n13.R)
  n15 = DynamicParameter() {parameterIndex=3, paramNames [DistortionNoise_PanU, DistortionNoise_PanV, "", ""]}
  n16 = TextureSample(UVs=Panner(Coordinate=Multiply(A=@n12, B=@n14), Time=Time(), Speed=AppendVector(A=@n15.DistortionNoise_PanU, B=@n15.DistortionNoise_PanU))) {samplerType=SAMPLERTYPE_Grayscale, texture=&LowResBlurredNoise}
  n17 = Multiply(A=AppendVector(A=@n16.R, B=@n16.R), B=ScalarParameter() {group=Primary Noise Distortion, parameterName=Distortion Mult, sortPriority=0}) {constB=0.2}
  n18 = DynamicParameter() {parameterIndex=2, paramNames [PrimaryNoise_PanU, PrimaryNoise_PanV, SecondaryNoise_PanU, SecondaryNoise_PanV]}
  n19 = TextureSample(UVs=Panner(Coordinate=Add(A=@n11, B=@n17), Time=Time(), Speed=AppendVector(A=@n18.PrimaryNoise_PanU, B=@n18.PrimaryNoise_PanU))) {texture=&AtmosphericCloudNoise01}
  n20 = ScalarParameter() {defaultValue=1, group=Primary Noise, parameterName=Primary Noise Power, sortPriority=0}
  n21 = Saturate(None=Multiply(
    A=Power(Base=@n19.RGB, Exp=@n20)
    B=ScalarParameter() {defaultValue=1, group=Primary Noise, parameterName=Primary Noise Mult, sortPriority=1}
   ))
  n22 = TextureCoordinate()
  n23 = VectorParameter() {group=Secondary Noise, parameterName=Secondary Noise UV Tile, sortPriority=2, defaultValue=<rgba 2 2 0 1>}
  n24 = AppendVector(A=@n23.R, B=@n23.R)
  n25 = DynamicParameter() {parameterIndex=2, paramNames [PrimaryNoise_PanU, PrimaryNoise_PanV, SecondaryNoise_PanU, SecondaryNoise_PanV]}
  n26 = TextureSample(UVs=Panner(Coordinate=Multiply(A=@n22, B=@n24), Time=Time(), Speed=AppendVector(A=@n25.SecondaryNoise_PanU, B=@n25.SecondaryNoise_PanU))) {samplerType=SAMPLERTYPE_LinearGrayscale, texture=&Snowfall_pick_Noise}
  n27 = ScalarParameter() {defaultValue=1, group=Secondary Noise, parameterName=Secondary Noise Power, sortPriority=0}
  n28 = Saturate(None=Multiply(
    A=Power(Base=@n26.RGB, Exp=@n27)
    B=ScalarParameter() {defaultValue=1, group=Secondary Noise, parameterName=Secondary Noise Mult, sortPriority=1}
   ))
  n29 = LinearInterpolate(Alpha=OneMinus(None=Multiply(A=@n21, B=@n28))) {constA=0.2}
  n30 = Constant3Vector() {constant=<rgba 0.5 0.4 0 0>}
  n31 = LinearInterpolate(Alpha=SphereMask(A=TextureCoordinate(), B=AppendVector(A=@n30.R, B=@n30.R), Radius=Constant() {r=0.5}, Hardness=Constant()) {attenuationRadius=2}) {constA=0.5, constB=1.25}
  n32 = TextureSampleParameter2D(UVs=Saturate(None=Multiply(A=@n29, B=@n31))) {parameterName=ColorGradient, texture=&/T_FX_RedPepperBackpack_Gradient}
  n33 = Multiply(A=@n32.RGB, B=DynamicParameter() {parameterIndex=1, paramNames [Rune_Blink, Rune_HDRMult, "", ""]}.Rune_HDRMult)
  n34 = MaterialFunctionCall(
   R=MaterialFunctionCall(Float3=MaterialFunctionCall(LDR Input=@n32.RGB, HDR Intensity Multiplier=@n33) {materialFunction=&DeriveHDRfromLDR}.RemappedRange) {materialFunction=&BreakOutFloat3Components}.R
   B=ScalarParameter() {defaultValue=0.01, parameterName=BlueAddition}
  ) {materialFunction=&AppendMany}
  n35 = DynamicParameter() {parameterIndex=1, paramNames [Rune_Blink, Rune_HDRMult, "", ""]}
  n36 = VectorParameter() {parameterName=OuterColor, defaultValue=<rgba 1 1 1 1>}
  n37 = VectorParameter() {parameterName=InnerColor, defaultValue=<rgba 0 0 0 1>}
  n38 = MaterialFunctionCall(
   Brightness=ScalarParameter() {defaultValue=1, parameterName=FresnelBrightness}
   Exponent=ScalarParameter() {defaultValue=1, parameterName=FresnelExp}
  ) {materialFunction=&MF_Fresnel_FX}
  MaterialFunctionCall(Emissive=LinearInterpolate(
    A=Multiply(A=@n34.RGB, B=@n35.Rune_Blink)
    B=LinearInterpolate(A=@n36.RGB, B=@n37.RGB, Alpha=@n38.Fresnel Double Sided)
    Alpha=ScalarParameter() {parameterName=UseFresnelColor}
   )) {materialFunction=&MF_FX_EmissiveAdjustments}
  Multiply()
  OneMinus(None=@n1)
 ]
}
