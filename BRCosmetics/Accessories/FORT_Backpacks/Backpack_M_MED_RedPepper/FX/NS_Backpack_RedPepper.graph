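# Niagara effect for the RedPepper backpack cosmetic that puts a flickering red glow on the backpack.
# An instant burst spawns persistent particles with looping lifetimes, positioned each update from the User.SkeletalMesh input with a small offset.
# Two mesh renderers, a light and a camera-offset sprite draw them.
# One irregular Blink curve drives dynamic material parameters, light color scale and faint sprite alpha, so all flicker in sync.
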NiagaraSystem /BRCosmetics/Accessories/FORT_Backpacks/Backpack_M_MED_RedPepper/FX/NS_Backpack_RedPepper
dependencies {usedDataInterfaces [&NiagaraDataInterfaceCurve], usedDynamicInputs [], usedModules [&SystemState]}
emitters {DirectionalBurst {
  emitterUpdateScript [
   &EmitterState()
   &SpawnBurst_Instantaneous(Use Spawn Probability=false, UseLoopCountLimit=true)
  ]
  particleSpawnScript [
   &InitializeParticle(
    Color=<rgba 1 0.05 0.05 1>
    Lifetime=6
    Mass Max=2
    Mass Mode=::NewEnumerator3
    Mesh Scale Mode=::NewEnumerator3
    Mesh Uniform Scale=2.5
    UsePositionOffset=false
    Write Material Random=true
   )
   &InitialMeshOrientation(Mesh Orientation Mode=::NewEnumerator2, Orientation Vector=<-1 0 0>, Use Rotation Vector=false)
  ]
  particleUpdateScript [
   &ParticleState(
    Kill Particles When Lifetime Has Elapsed=false
    Let Infinitely Lived Particles Die When Emitter Deactivates=true
    Loop Particles Lifetime=true
   )
   &SkeletalMeshLocation(
    MaskSpawnGroup=false
    Mesh Sampling Type=::NewEnumerator5
    Sampled Position Offset=<0 5 0>
    Skeletal Mesh=$User.SkeletalMesh
    Socket Sampling Mode=::NewEnumerator5
   )
   /SetVariables(Particles.Blink:NiagaraFloat=&FloatFromCurve(FloatCurve=DataInterface() {propertyValues {Curve {keys [
         {arriveTangent=-1.6748, interpMode=RCIM_Cubic, time=0, value=1}
         {arriveTangent=1.6386, interpMode=RCIM_Cubic, leaveTangent=1.6386, time=0.0746, value=0.875}
         {arriveTangent=-0.0227, interpMode=RCIM_Cubic, leaveTangent=-0.0227, tangentMode=RCTM_User, time=0.1526, value=1.25}
         {arriveTangent=-2.5347, interpMode=RCIM_Cubic, leaveTangent=-2.5347, time=0.1912, value=0.5}
         {arriveTangent=1.3845, interpMode=RCIM_Cubic, leaveTangent=1.3845, time=0.2748, value=0.9402}
         {arriveTangent=2.22, interpMode=RCIM_Cubic, leaveTangent=2.22, time=0.3357, value=0.7}
         {arriveTangent=-1.1092, interpMode=RCIM_Cubic, leaveTangent=-1.1092, time=0.3918, value=1.2}
         {arriveTangent=-2.7557, interpMode=RCIM_Cubic, leaveTangent=-2.7557, time=0.4259, value=0.6}
         {arriveTangent=-0.7763, interpMode=RCIM_Cubic, leaveTangent=-0.7763, time=0.4871, value=0.9373}
         {arriveTangent=0.3027, interpMode=RCIM_Cubic, leaveTangent=0.3027, time=0.5225, value=0.525}
         {arriveTangent=2.2304, interpMode=RCIM_Cubic, leaveTangent=2.2304, time=0.6112, value=0.9748}
         {arriveTangent=2.9599, interpMode=RCIM_Cubic, leaveTangent=2.9599, time=0.69, value=0.8986}
         {arriveTangent=-2.5142, interpMode=RCIM_Cubic, leaveTangent=-2.5142, time=0.738, value=1.35}
         {arriveTangent=-3.4304, interpMode=RCIM_Cubic, leaveTangent=-3.4304, time=0.7889, value=0.65}
         {arriveTangent=-0.8754, interpMode=RCIM_Cubic, leaveTangent=-0.8754, time=0.8674, value=0.9059}
         {arriveTangent=0.71, interpMode=RCIM_Cubic, leaveTangent=0.71, time=0.9031, value=0.55}
         <at 1 1 RCIM_Cubic>
        ]}}})) {isSetParameters=true}
   &DynamicMaterialParameters(
    Index 0 Param 1=&Multiply_Float(A=$Particles.Blink, B=0.2)
    Index 0 Param 2=30
    Index 1 Param 1=&Multiply_Float(A=$Particles.Blink, B=1)
    Index 1 Param 2=1
    Index 2 Param 1=0.025
    Index 2 Param 2=0.05
    Index 2 Param 3=0.025
    Index 2 Param 4=0.05
    Index 3 Param 2=0.02
    Index 3 Param 3=2
    Index 3 Param 4=2
    Module.Param0WriteEnabled=true
    Module.Param0WriteEnabled1=true
    Module.Param0WriteEnabled2=true
    Module.Param0WriteEnabled3=true
    Module.Param1WriteEnabled=true
    Module.Param1WriteEnabled1=true
    Module.Param1WriteEnabled2=true
    Module.Param1WriteEnabled3=true
    Module.Param2WriteEnabled=false
    Module.Param2WriteEnabled1=false
    Module.Param2WriteEnabled2=true
    Module.Param2WriteEnabled3=false
    Module.Param3WriteEnabled=false
    Module.Param3WriteEnabled1=false
    Module.Param3WriteEnabled2=true
    Module.Param3WriteEnabled3=false
    Write Parameter Index 1=true
    Write Parameter Index 2=true
    Write Parameter Index 3=true
   )
   &FN_Light_Attributes(
    Color=$Particles.Color
    Create Color Attribute=true
    Create Exponent Attribute=true
    Create Radius Attribute=true
    Enable Particle Light Enable=true
    Exponent=3
    Light Enabled=true
    Radius=30
    Scale Color=&Multiply_Float(A=$Particles.Blink, B=1)
   )
   &FN_Sprite_Attributes(
    Camera Offset=20
    Color=&MultiplyLinearColor(
     Linear Color=$Particles.Color
     Scale Factor=&MakeVector4(X=1, Y=1, Z=1, W=&Multiply_Float(A=0.05, B=$Particles.Blink))
    )
    Create Camera Offset Attribute=true
    Create Color Attribute=true
    Write to Module Namespace=true
   )
  ]
  rendererClasses [
   &NiagaraMeshRendererProperties
   &NiagaraLightRendererProperties
   &NiagaraSpriteRendererProperties
  ]
  renderers [
   {rendererClass=&NiagaraMeshRendererProperties, rendererIndex=0}
   {rendererClass=&NiagaraMeshRendererProperties, rendererIndex=1}
   {rendererClass=&NiagaraLightRendererProperties, rendererIndex=2}
   {rendererClass=&NiagaraSpriteRendererProperties, rendererIndex=3}
  ]
 }}
summary {
 systemName=NS_Backpack_RedPepper
 emitters [{
   emitterName=DirectionalBurst
   rendererClasses [
    &NiagaraMeshRendererProperties
    &NiagaraLightRendererProperties
    &NiagaraSpriteRendererProperties
   ]
  }]
 userVariables [User.SkeletalMesh {
   type=NiagaraDataInterfaceSkeletalMesh
   defaultValue {
    struct=NiagaraExt_VariableValue_DataInterface
    value {dataInterface=&self:NiagaraDataInterfaceSkeletalMesh_0}
   }
  }]
}
